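# Physics asset for the Lion Soldier medium male backpack cosmetic, defining the bodies and joints for its dangling pouch straps and center pouch.
# The pack bone carries several boxes approximating the bag, and small box bodies cover each DYN_ strap segment and the center pouch.
# Angle-limited constraints chain strap segments back to the pack, with mostly tight roots, looser links between segments, and low twist throughout.
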PhysicsAsset /BRCosmetics/Accessories/FORT_Backpacks/Backpack_M_MED_Lion_Soldier/Meshes/M_MED_Lion_Soldier_Backpack_Physics
bodies {
 DYN_PouchCenter [{shapeType=Box, extentX=1.375, extentY=5.1875, extentZ=8.375, center=<0.5943 2.8084 0>}]
 DYN_PouchFrontL_Strap01 [{shapeType=Box, extentX=1.75, extentY=1, extentZ=2.625, center=<0.9639 0.0001 0>, rotation=<rot 0 0 -8.4397>}]
 DYN_PouchFrontL_Strap02 [{shapeType=Box, extentX=1.375, extentY=1, extentZ=2.5625, center=<0.6902 0 0>}]
 DYN_PouchFrontL_StrapL01 [{shapeType=Box, extentX=2.375, extentY=1, extentZ=3.25, center=<1.251 0.3054 0>, rotation=<rot 0 -8.4393 0>}]
 DYN_PouchFrontL_StrapL02 [{shapeType=Box, extentX=2.3125, extentY=1, extentZ=3.1875, center=<1.1664 -0.2208 0>}]
 DYN_PouchFrontL_StrapL03 [{shapeType=Box, extentX=2.125, extentY=1, extentZ=3.25, center=<1.1406 0.3747 0.0001>}]
 DYN_PouchFrontR_Strap01 [{shapeType=Box, extentX=1.8125, extentY=1, extentZ=2.5625, center=<0.9866 0 0>}]
 DYN_PouchFrontR_Strap02 [{shapeType=Box, extentX=1.5, extentY=1, extentZ=2.4375, center=<0.8123 0 0>}]
 DYN_PouchFrontR_StrapL01 [{shapeType=Box, extentX=2.5, extentY=1, extentZ=3.375, center=<1.3903 0.2948 0>}]
 DYN_PouchFrontR_StrapL02 [{shapeType=Box, extentX=1.5625, extentY=1, extentZ=3.5, center=<0.8727 0.3239 0>}]
 DYN_PouchFrontR_StrapL03 [{shapeType=Box, extentX=1.9375, extentY=1, extentZ=3.4375, center=<1.1002 0.4487 0>}]
 DYN_PouchL_Strap01 [{shapeType=Box, extentX=1.9375, extentY=0.8125, extentZ=3, center=<0.922 0 0>}]
 DYN_PouchL_Strap02 [{shapeType=Box, extentX=1.25, extentY=0.8125, extentZ=3, center=<0.6446 0 0>}]
 DYN_PouchL_Strap03 [{shapeType=Box, extentX=1.875, extentY=1, extentZ=3, center=<0.9706 0.0001 0>}]
 DYN_PouchR_Strap01 [{shapeType=Box, extentX=2.125, extentY=1, extentZ=3, center=<1.1066 0 0>}]
 DYN_PouchR_Strap02 [{shapeType=Box, extentX=0.875, extentY=1, extentZ=3, center=<0.423 0 0>}]
 DYN_PouchR_Strap03 [{shapeType=Box, extentX=1.625, extentY=1, extentZ=2.9375, center=<0.812 0 0>}]
 pack [
  {shapeType=Box, extentX=1, extentY=1, extentZ=1}
  {shapeType=Box, extentX=7.5, extentY=5.75, extentZ=7.5, center=<10.5949 -3.8958 -13.0223>, rotation=<rot 0 -16.8813 0>}
  {shapeType=Box, extentX=7.5, extentY=6.4375, extentZ=7.5, center=<-10.5778 -5.0106 -12.6636>}
  {shapeType=Box, extentX=5.375, extentY=7.5, extentZ=7.5, center=<-6.4535 -10.551 -15.1597>, rotation=<rot -0.2756 -5.6186 -2.7994>}
  {shapeType=Box, extentX=6.4375, extentY=7.5, extentZ=7.5, center=<4.7418 -10.5718 -14.7971>}
  {shapeType=Box, extentX=11.6875, extentY=4.6875, extentZ=9.1875, center=<-1.2141 -10.6256 -6.0743>}
 ]
}
constraints [
 {bone1Name=DYN_PouchR_Strap02, bone2Name=DYN_PouchR_Strap01, swing1LimitDegrees=45, swing2LimitDegrees=45, twistLimitDegrees=5}
 {bone1Name=DYN_PouchR_Strap03, bone2Name=DYN_PouchR_Strap02, swing1LimitDegrees=45, swing2LimitDegrees=45, twistLimitDegrees=5}
 {bone1Name=DYN_PouchL_Strap02, bone2Name=DYN_PouchL_Strap01, swing1LimitDegrees=45, swing2LimitDegrees=45, twistLimitDegrees=5}
 {bone1Name=DYN_PouchL_Strap03, bone2Name=DYN_PouchL_Strap02, swing1LimitDegrees=45, swing2LimitDegrees=45, twistLimitDegrees=5}
 {bone1Name=DYN_PouchFrontL_Strap02, bone2Name=DYN_PouchFrontL_Strap01, swing1LimitDegrees=45, swing2LimitDegrees=45, twistLimitDegrees=5}
 {bone1Name=DYN_PouchFrontL_StrapL02, bone2Name=DYN_PouchFrontL_StrapL01, swing1LimitDegrees=50, swing2LimitDegrees=20, twistLimitDegrees=5}
 {bone1Name=DYN_PouchFrontL_StrapL03, bone2Name=DYN_PouchFrontL_StrapL02, swing1LimitDegrees=50, swing2LimitDegrees=20, twistLimitDegrees=5}
 {bone1Name=DYN_PouchFrontR_StrapL02, bone2Name=DYN_PouchFrontR_StrapL01, swing1LimitDegrees=50, swing2LimitDegrees=20, twistLimitDegrees=5}
 {bone1Name=DYN_PouchFrontR_StrapL03, bone2Name=DYN_PouchFrontR_StrapL02, swing1LimitDegrees=50, swing2LimitDegrees=20, twistLimitDegrees=5}
 {bone1Name=DYN_PouchFrontR_Strap02, bone2Name=DYN_PouchFrontR_Strap01, swing1LimitDegrees=45, swing2LimitDegrees=45, twistLimitDegrees=5}
 {bone1Name=DYN_PouchR_Strap01, bone2Name=pack, swing1LimitDegrees=10, swing2LimitDegrees=1, twistLimitDegrees=5}
 {bone1Name=DYN_PouchL_Strap01, bone2Name=pack, swing1LimitDegrees=10, swing2LimitDegrees=1, twistLimitDegrees=5}
 {bone1Name=DYN_PouchFrontL_Strap01, bone2Name=pack, swing1LimitDegrees=10, swing2LimitDegrees=1, twistLimitDegrees=5}
 {bone1Name=DYN_PouchFrontL_StrapL01, bone2Name=pack, swing1LimitDegrees=75, swing2LimitDegrees=30, twistLimitDegrees=5}
 {bone1Name=DYN_PouchFrontR_StrapL01, bone2Name=pack, swing1LimitDegrees=50, swing2LimitDegrees=20, twistLimitDegrees=5}
 {bone1Name=DYN_PouchFrontR_Strap01, bone2Name=pack, swing1LimitDegrees=10, swing2LimitDegrees=1, twistLimitDegrees=5}
 {bone1Name=DYN_PouchCenter, bone2Name=pack, swing1LimitDegrees=30, swing2LimitDegrees=1, twistLimitDegrees=1}
]
